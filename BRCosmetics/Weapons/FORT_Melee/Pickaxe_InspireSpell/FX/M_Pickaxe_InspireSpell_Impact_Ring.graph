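# Unlit translucent Niagara sprite material, named as the impact ring for the InspireSpell pickaxe.
# Two smoke pattern samples pan over radial coordinates at a dynamic-parameter speed and are added, then clamped.
# That mask blends particle color toward a multiplied, by default darker, tint before MF_FX_EmissiveAdjustments.
# A ground-ring texture's red channel multiplied by particle alpha forms the ring shape.
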Material /BRCosmetics/Weapons/FORT_Melee/Pickaxe_InspireSpell/FX/M_Pickaxe_InspireSpell_Impact_Ring
graph {
 properties {
  bUsedWithNanite=true
  bUsedWithNiagaraSprites=true
  blendMode=BLEND_Translucent
  pixelDepthOffsetMode=PDOM_Legacy
  shadingModel=MSM_Unlit
  usedShadingModels=Unlit
 }
 nodes [
  n1 = ParticleColor()
  n2 = ScalarParameter() {defaultValue=0.1, parameterName=Additional Color Multiplier}
  n3 = MaterialFunctionCall(Vector or UVs=Subtract(A=TextureCoordinate() {uTiling=0.2, vTiling=0.2}) {constB=0.1}) {materialFunction=&VectorToRadialValue}
  n4 = Multiply(A=Time(), B=DynamicParameter() {defaultValue=<rgba 1 0.1 0 1>, paramNames [Pan Speed, Additional Color Multiplier, Param3, Power]}.Pan Speed)
  n5 = Add(
   A=TextureSample(UVs=Panner(Coordinate=@n3.Radial Coordinates, Time=@n4) {speedX=0.04, speedY=-0.55}) {texture=/Game/Global/Textures/Default/Tileable_Patterns/T_Blowing_Smoke_Pattern2_Tile_DISP}.RGB
   B=TextureSample(UVs=Panner(Coordinate=@n3.Radial Coordinates, Time=@n4) {speedY=-0.42}) {texture=&T_Blowing_Smoke_Pattern2_Tile_DISP}.RGB
  )
  MaterialFunctionCall(Emissive=LinearInterpolate(A=@n1.RGB, B=Multiply(A=@n1.RGB, B=@n2), Alpha=Clamp(None=@n5))) {materialFunction=&MF_FX_EmissiveAdjustments}
  Multiply(A=@n1.A, B=TextureSample() {texture=/Game/Effects/Fort_Effects/Textures/Outlander/T_InTheZone_GroundRing_01}.R)
 ]
}
